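# Sound cue for the GreenGlop item's impact sound, classed under projectile effects.
# A random node picks one of two reused coconut impact waves, and a modulator pitches it well up before a mixer at reduced input volume, whose other input is empty.
# The cue plays quietly through the foraged-item bounce attenuation asset, and overridden concurrency allows one instance per owner, stopping the oldest.
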SoundCue /Figment_S10/Athena/Sounds/Items/GreenGlop/Athena_Greenglop_Impact_Cue
root=Mixer_0
nodes [
 {id=Mixer_0, children [null,Modulator_0], properties {inputVolume [1,0.5]}}
 {id=Modulator_0, children [Random_1], properties {pitchMax=1.6, pitchMin=1.5}}
 {id=Random_1, children [WavePlayer_5,WavePlayer_6], properties {weights [1,1]}}
 {id=WavePlayer_5, properties {soundWaveAssetPtr=/Game/Athena/Sounds/Items/JungleForaged/Athena_Coconut_Impact_01}}
 {id=WavePlayer_6, properties {soundWaveAssetPtr=/Game/Athena/Sounds/Items/JungleForaged/Athena_Coconut_Impact_02}}
]
properties {
 attenuationSettings=/Figment_S10/Sounds/Attenuation/Fort_ForagedItem_Bounce_Att
 bOverrideConcurrency=true
 maxDistance=3000
 soundClassObject=/Figment_S10/Sounds/FortSoundClassesAndModes/Fort_ProjectileEffects
 volumeMultiplier=0.25
 attenuationOverrides {
  bAttenuate=false
  bAttenuateWithLPF=true
  bEnableListenerFocus=true
  falloffDistance=1536
  focusAzimuth=15
  lPFFrequencyAtMax=1000
  lPFRadiusMax=500
  lPFRadiusMin=0
  nonFocusAzimuth=30
  nonFocusVolumeAttenuation=0.5
  attenuationShapeExtents {x=0}
 }
 concurrencyOverrides {bLimitToOwner=true, maxCount=1, resolutionRule=StopOldest}
}
