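# Sound cue for the GreenGlop item's first impact, layering a glop sound with a coconut hit.
# A mixer combines a random pick of five glop impacts with a quieter random pick of two borrowed coconut impacts, each through its own slight pitch randomization.
# It plays at reduced volume through the projectile effects sound class and foraged-item bounce attenuation, with a concurrency override capping simultaneous instances.
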SoundCue /Figment_S10/Athena/Sounds/Items/GreenGlop/Athena_Greenglop_Impact_First_Cue
root=Mixer_0
nodes [
 {id=Mixer_0, children [Modulator_5,Modulator_0], properties {inputVolume [1,0.5]}}
 {id=Modulator_0, children [Random_1], properties {pitchMax=1.1, pitchMin=0.9}}
 {id=Modulator_5, children [Random_0], properties {pitchMax=1.1, pitchMin=0.9}}
 {id=Random_0, children [WavePlayer_0,WavePlayer_1,WavePlayer_2,WavePlayer_3,WavePlayer_4], properties {weights [1,1,1,1,1]}}
 {id=Random_1, children [WavePlayer_5,WavePlayer_6], properties {weights [1,1]}}
 {id=WavePlayer_0, properties {soundWaveAssetPtr=&/GreenGlop_Impact_01}}
 {id=WavePlayer_1, properties {soundWaveAssetPtr=&/GreenGlop_Impact_02}}
 {id=WavePlayer_2, properties {soundWaveAssetPtr=&/GreenGlop_Impact_03}}
 {id=WavePlayer_3, properties {soundWaveAssetPtr=&/GreenGlop_Impact_04}}
 {id=WavePlayer_4, properties {soundWaveAssetPtr=&/GreenGlop_Impact_05}}
 {id=WavePlayer_5, properties {soundWaveAssetPtr=/Game/Athena/Sounds/Items/JungleForaged/Athena_Coconut_Impact_01}}
 {id=WavePlayer_6, properties {soundWaveAssetPtr=/Game/Athena/Sounds/Items/JungleForaged/Athena_Coconut_Impact_02}}
]
properties {
 attenuationSettings=/Figment_S10/Sounds/Attenuation/Fort_ForagedItem_Bounce_Att
 bOverrideConcurrency=true
 maxDistance=3000
 soundClassObject=/Figment_S10/Sounds/FortSoundClassesAndModes/Fort_ProjectileEffects
 volumeMultiplier=0.25
 attenuationOverrides {falloffDistance=1536, attenuationShapeExtents {x=0}}
 concurrencyOverrides {maxCount=3}
}
